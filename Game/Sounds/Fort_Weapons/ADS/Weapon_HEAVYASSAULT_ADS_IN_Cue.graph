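# Sound cue for the heavy assault weapon's aim-down-sights-in foley, in the weapon foley sound class.
# A Branch on the IsLocal parameter plays sound only when true; the false and unset inputs are empty, so those cases stay silent.
# The true path runs through Attenuation and Modulator nodes into an equally weighted Random pick between two wave variants, excluded from branch culling.
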SoundCue /Game/Sounds/Fort_Weapons/ADS/Weapon_HEAVYASSAULT_ADS_IN_Cue
root=Branch_0
nodes [
 {id=Attenuation_0, children [Modulator_0]}
 {id=Branch_0, children [Attenuation_0, null, null], properties {boolParameterName=IsLocal}}
 {id=Modulator_0, children [Random_0]}
 {id=Random_0, children [WavePlayer_1,WavePlayer_2], properties {bShouldExcludeFromBranchCulling=true, weights [1,1]}}
 {id=WavePlayer_1, properties {soundWaveAssetPtr=&/Weapon_HEAVYASSAULT_ADS_IN_01}}
 {id=WavePlayer_2, properties {soundWaveAssetPtr=&/Weapon_HEAVYASSAULT_ADS_IN_02}}
]
properties {maxDistance=2097152, soundClassObject=&Fort_WeaponFoley, volumeMultiplier=0.376}
