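# Sound cue for military rifle aim-down-sights entry foley, routed to the Fort_WeaponFoley sound class.
# A single wave player plays the Weapon_MILRIFLE_ADS_IN_02 wave through a Modulator left at default settings, which applies slight random pitch and volume variation on each play.
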SoundCue /Game/Sounds/Fort_Weapons/ADS/Weapon_MILRIFLE_ADS_IN_Cue
root=Modulator_1
nodes [
 {id=Modulator_1, children [WavePlayer_2]}
 {id=WavePlayer_2, properties {soundWaveAssetPtr=&/Weapon_MILRIFLE_ADS_IN_02}}
]
properties {maxDistance=2097152, soundClassObject=&Fort_WeaponFoley, volumeMultiplier=1}
