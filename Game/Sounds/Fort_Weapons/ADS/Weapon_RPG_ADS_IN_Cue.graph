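# Sound cue for the RPG weapon's aim-down-sights-in foley.
# A single wave player plays one RPG ADS-in sound through a modulator left at its default settings.
# The cue routes to the weapon foley sound class at reduced volume and uses the shared weapon reload attenuation for distance falloff.
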SoundCue /Game/Sounds/Fort_Weapons/ADS/Weapon_RPG_ADS_IN_Cue
root=Modulator_0
nodes [
 {id=Modulator_0, children [WavePlayer_0]}
 {id=WavePlayer_0, properties {soundWaveAssetPtr=&/Weapon_RPG_ADS_IN_01}}
]
properties {
 attenuationSettings=&Fort_Weapon_Reload_Att
 maxDistance=2250
 soundClassObject=&Fort_WeaponFoley
 volumeMultiplier=0.376
}
